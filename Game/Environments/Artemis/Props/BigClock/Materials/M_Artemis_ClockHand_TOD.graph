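# Nanite-enabled material for the Artemis big clock's hands, rotating them to show a time.
# A ClockTime custom primitive data value doubles into 12-hour turns; the isHourHand switch picks smooth hour rotation or 12x minute rotation floored into 60 steps per turn.
# RotateAboutAxis turns vertices around local Y at the object pivot, with FixRotateAboutAxisNormals correcting the normal map.
# Despite the TOD name, the TimeOfDay_0-1 collection parameter feeds nothing.
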Material /Game/Environments/Artemis/Props/BigClock/Materials/M_Artemis_ClockHand_TOD
graph {
 properties {bUsedWithNanite=true, pixelDepthOffsetMode=PDOM_Legacy, usedShadingModels=Default Lit}
 nodes [
  CollectionParameter() {collection=&FortniteMaterialParameters, parameterName=TimeOfDay_0-1}
  n1 = Transform(None=Constant3Vector() {constant=<rgba 0 1 0 0>}) {transformSourceType=TRANSFORMSOURCE_Local}
  n2 = Multiply(A=ScalarParameter() {bUseCustomPrimitiveData=true, parameterName=ClockTime}) {constB=2, desc=Convert to 0 - 12}
  n3 = Constant() {r=60}
  n4 = Multiply(A=Multiply(A=@n2) {constB=12, desc=Minute hand has 12 Rotations per hour}, B=@n3) {constB=60}
  n5 = StaticSwitchParameter(True=@n2, False=Divide(A=Floor(None=@n4), B=@n3) {constB=60}) {parameterName=isHourHand}
  MaterialFunctionCall(
   Rotation Axis=@n1
   Rotation Angle=@n5
   World Space Vertex Normals=Transform(None=TextureSample() {samplerType=SAMPLERTYPE_Normal, texture=/Game/Environments/Artemis/Props/BigClock/Textures/T_Artemis_BigClock_N}.RGB)
  ) {materialFunction=&FixRotateAboutAxisNormals}
  RotateAboutAxis(
   NormalizedRotationAxis=@n1
   RotationAngle=@n5
   PivotPoint=MaterialFunctionCall() {materialFunction=&ObjectPivotPoint}.Object Pivot Location
   Position=WorldPosition().XYZ
  )
  TextureSample() {texture=/Game/Environments/Artemis/Props/BigClock/Textures/T_Artemis_BigClock_D}
  TextureSample() {samplerType=SAMPLERTYPE_Masks, texture=/Game/Environments/Artemis/Props/BigClock/Textures/T_Artemis_BigClock_S}
 ]
}
